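; Gameplay ability for quest S11 OT2_07 that grants progress when the player swims far enough within a recent time window.
; On activation, a looping timer samples the avatar's 2D movement into a circular array, maintaining a rolling distance total without re-summing.
; Reaching the threshold sends a custom stat event to the player's quest manager and ends the ability.
; A change to walking also ends it, and ending clears the timer.
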(event Ability|EventActivateAbilityFromEvent
  (params (EventData :type Gameplay%20Event%20Data%20Structure%20%28by%20ref%29))
  (|CallParentFunction EventData)
  (Ability|Tasks|WaitMovementModeChange "MOVE_Walking"
    (:then
      (Variables|Default|SetCurrentLocation (Transformation|GetActorLocation (Ability|GetAvatarActorfromActorInfo)))
      (Variables|Default|SetDistanceArray (Utilities|Array|MakeArray 0.0 0.0 0.0 0.0 0.0))
      (Variables|Default|SetDistanceArrayIndex 0)
      (Variables|Default|SetRollingDistanceTraveled 0.0)
      (bind _returnvalue (Utilities|Time|SetTimerbyEvent (delegate Custom|TimedDistanceCheck) (Variables|Default|GetTimedDistanceCheckInterval) true))
      (Variables|Default|SetTimedDistanceCheckHandle _returnvalue))
    (:OnChange
      (Ability|EndAbility))))

(event Custom|TimedDistanceCheck
  (bind _oldlocation (Variables|Default|GetOldLocation))
  (bind _currentlocation (Variables|Default|GetCurrentLocation))
  (bind _returnvalue (Math|Vector|Distance2D(Vector) _currentlocation _oldlocation))
  (bind _distancearray (Variables|Default|GetDistanceArray))
  (bind _distancearrayindex (Variables|Default|GetDistanceArrayIndex))
  ;[1 1 1] Keeps a rolling count of how far has swam in the past X seconds. Grants quest progress if that rolling distance has exceeded a threshold.
  (Variables|Default|SetOldLocation (Variables|Default|GetCurrentLocation))
  (Variables|Default|SetCurrentLocation (Transformation|GetActorLocation (Ability|GetAvatarActorfromActorInfo)))
  (Variables|Default|SetRollingDistanceTraveled (- (Variables|Default|GetRollingDistanceTraveled) (Utilities|Array|Get(aref) _distancearray _distancearrayindex)))
  (Utilities|Array|SetArrayElem _distancearray _distancearrayindex _returnvalue true)
  (Variables|Default|SetRollingDistanceTraveled (+ (Variables|Default|GetRollingDistanceTraveled) _returnvalue))
  (Variables|Default|SetDistanceArrayIndex (Math|Integer|%25%28Integer%29 (+ (Variables|Default|GetDistanceArrayIndex) 1) (Variables|Default|GetDistanceArraySize)))
  (if (>= (Variables|Default|GetRollingDistanceTraveled) (Variables|Default|GetQuestGrantThreshold))
    (Quest|SendComplexCustomStatEvent (Variables|GAVars|Data|GetOwningPlayerQuestManager) 0 "(GameplayTags=)" (Variables|Default|GetTargetTags) 1)
    (Ability|EndAbility)))

(event Ability|EventOnEndAbility
  (params (bWasCancelled :type Boolean))
  (Utilities|Time|ClearandInvalidateTimerbyHandle (Variables|Default|GetTimedDistanceCheckHandle)))
